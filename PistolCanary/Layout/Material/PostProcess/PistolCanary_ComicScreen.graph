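# PistolCanary_ComicScreen is a post-process material that drains color from the rendered screen for a comic effect.
# It feeds the PostProcessInput0 scene color into a Desaturation node.
# The ComicEffect scalar in the PistolCanary_MPC parameter collection sets the desaturation fraction, so the strength is controlled globally from outside the material.
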Material /PistolCanary/Layout/Material/PostProcess/PistolCanary_ComicScreen
graph {
 properties {bAutomaticallySetUsageInEditor=false, materialDomain=MD_PostProcess, usedShadingModels=Default Lit}
 nodes [Desaturation(
   None=SceneTexture() {sceneTextureId=PPI_PostProcessInput0}.Color
   Fraction=CollectionParameter() {collection=&/PistolCanary_MPC, parameterName=ComicEffect}
  )]
}
